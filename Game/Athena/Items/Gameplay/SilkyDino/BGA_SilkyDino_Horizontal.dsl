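; SilkyDino horizontal gameplay actor that grants a "Faucet" gameplay effect, apparently healing, to pawns inside its capsule.
; Once playlist data is ready, it self-destructs after a random delay if at or below ocean level, otherwise binds capsule overlaps.
; Overlap start applies the effect unless the pawn has the vehicle-check tag; overlap end removes effects tagged SilkyDino Heal.
; The construction script sets Falls and FallsBase heights from variables.
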(event Custom|OnComponentEndOverlap_Event_0
  (params (OverlappedComponent :type Primitive%20Component%20Object%20Reference) (OtherActor :type Actor%20Object%20Reference) (OtherComp :type Primitive%20Component%20Object%20Reference) (OtherBodyIndex :type Integer))
  (bind _asfort_pawn (Utilities|Casting|CastToFortPawn OtherActor))
  (GameplayEffects|RemoveActiveEffectswithTags (Ability|GetAbilitySystemComponent _asfort_pawn) "(GameplayTags=((TagName=\"Asset.Silky.SilkyDino.Heal\")))"))

(event EventBeginPlay
  (bind (_returnvalue_x _returnvalue_y _returnvalue_z) (Math|Vector|BreakVector (Transformation|GetActorLocation)))
  (Class|FortAsyncActionCurrentPlaylistReady|PlaylistDataReadyAsync
    (:OnReady
      (if (or (Fortnite|WaterHelpers|IsActorSubmergedinOcean self) (<= _returnvalue_z (Fortnite|WaterHelpers|GetOceanTotalHeight)))
        (Utilities|FlowControl|Delay (Math|Random|RandomFloatinRange 1.0 1e+01))
        (Actor|DestroyActor)
        (else
          (Collision|AssignOnComponentBeginOverlap (Variables|Default|GetCapsule) (delegate Custom|OnComponentBeginOverlap_Event_0))
          (Collision|AssignOnComponentEndOverlap (Variables|Default|GetCapsule) (delegate Custom|OnComponentEndOverlap_Event_0)))))))

(event Custom|OnComponentBeginOverlap_Event_0
  (params (OverlappedComponent :type Primitive%20Component%20Object%20Reference) (OtherActor :type Actor%20Object%20Reference) (OtherComp :type Primitive%20Component%20Object%20Reference) (OtherBodyIndex :type Integer) (bFromSweep :type Boolean) (SweepResult :type Hit%20Result%20Structure%20%28by%20ref%29))
  (bind _asfort_pawn (Utilities|Casting|CastToFortPawn OtherActor))
  (if (not (GameplayTags|HasMatchingGameplayTag _asfort_pawn (Variables|Default|GetVehicle%20Tag%20Check)))
    (GameplayEffects|FortApplyGameplayEffectToTarget :self (Ability|GetFortAbilitySystemComponent self) :GameplayEffectClass (Variables|Default|GetGE%20Faucet) :TargetOfEffect (Ability|GetFortAbilitySystemComponent _asfort_pawn) :OptionalSourceObject self :OptionalEffectCauser self :Level 0.0)))

(fn ConstructionScript
  (Transformation|SetRelativeLocation (Variables|Default|GetFalls) (Math|Vector|MakeVector (.x (Variables|/Script/Engine.SceneComponent|GetRelativeLocation (Variables|Default|GetFalls))) (.y (Variables|/Script/Engine.SceneComponent|GetRelativeLocation (Variables|Default|GetFalls))) (.z (Variables|Default|GetFX_FallsHeight))) false false)
  (Transformation|SetRelativeLocation (Variables|Default|GetFallsBase) (Math|Vector|MakeVector (.x (Variables|/Script/Engine.SceneComponent|GetRelativeLocation (Variables|Default|GetFallsBase))) (.y (Variables|/Script/Engine.SceneComponent|GetRelativeLocation (Variables|Default|GetFallsBase))) (.z (Variables|Default|GetFX_BaseHeight))) false false))
